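# Translucent unlit Niagara material for sprites, ribbons and meshes, by its name the WiseOtter long bullet effect.
# Emissive layers sum hue-shifted inky smoke masked by a panning trail, scrolling smoke remapped through a blue curve-atlas row, and a noise-broken, tinted spherical-mask glow.
# The alpha chain multiplies particle alpha by sine and sphere shape masks, subtracting a Dissolve value.
# Dynamic parameters supply that Dissolve and the UScale that stretches texture U tiling per particle.
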Material /WiseOtter/Assets/FX/M_NDC_WiseOtter_Bullet_Long
graph {
 properties {
  bAutomaticallySetUsageInEditor=false
  bUsedWithNiagaraMeshParticles=true
  bUsedWithNiagaraRibbons=true
  bUsedWithNiagaraSprites=true
  blendMode=BLEND_Translucent
  shadingModel=MSM_Unlit
  usedShadingModels=Unlit
 }
 nodes [
  ComponentMask(None=ParticleSize()) {g=true}
  Constant() {r=1}
  n1 = DynamicParameter() {defaultValue=<rgba 2 0 1 1>, paramNames [UScale,Dissolve,Param3,Param4]}
  n2 = Constant() {r=1}
  n3 = TextureSample(UVs=Multiply(A=TextureCoordinate() {uTiling=0.7, vTiling=2}, B=AppendVector(A=@n1.UScale, B=@n2))) {texture=&T_Inky_Smoke_Tile}
  n4 = MaterialFunctionCall(
   Hue Shift Percentage=Add(A=LinearInterpolate(Alpha=@n3.R) {constB=-0.4}, B=ScalarParameter() {defaultValue=-0.01, parameterName=Color offset}) {constB=0}
   Texture=VectorParameter() {parameterName=HueShiftColor, defaultValue=<rgba 0.0257 0.3504 1 1>}.RGB
  ) {materialFunction=&HueShift}
  n5 = DynamicParameter() {defaultValue=<rgba 2 0 1 1>, paramNames [UScale,Dissolve,Param3,Param4]}
  n6 = Constant() {r=1}
  n7 = Panner(Coordinate=Multiply(A=TextureCoordinate() {uTiling=0.5}, B=AppendVector(A=@n5.UScale, B=@n6))) {speedX=-1}
  n8 = Multiply(
   A=Desaturation(None=Multiply(A=@n4.Result), Fraction=Constant() {r=0.2})
   B=Power(Base=TextureSample(UVs=@n7) {texture=/Game/Weapons/FORT_Melee/Pickaxe_Future_Samurai_Summer_Male/FX/T_Pickaxe_Future_Samurai_Summer_Male_Trail_02}.G) {constExponent=0.7}
  )
  n9 = TextureCoordinate() {uTiling=1.3}
  n10 = AppendVector(A=DynamicParameter() {defaultValue=<rgba 2 0 1 1>, paramNames [UScale,Dissolve,Param3,Param4]}.UScale, B=Constant() {r=1})
  n11 = TextureSample(UVs=Panner(Coordinate=Multiply(A=@n9, B=@n10)) {speedX=-1}) {samplerType=SAMPLERTYPE_Grayscale, texture=&T_Smoke_Trail_Soft}
  n12 = TextureCoordinate() {uTiling=1.5}
  n13 = TextureSample(UVs=Panner(Coordinate=Multiply(A=@n12, B=@n10)) {speedX=-1.5}) {texture=/Game/Weapons/FORT_Melee/Pickaxe_Future_Samurai_Summer_Male/FX/T_Pickaxe_Future_Samurai_Summer_Male_Trail_03}
  n14 = CurveAtlasRowParameter(CurveTime=LinearInterpolate(A=Multiply(A=@n11.R, B=@n11.R), B=Power(Base=@n13.R) {constExponent=1})) {
   atlas=/WiseOtter/Assets/FX/Unibeam/CA_WiseOtter
   curve=/WiseOtter/Assets/FX/Unibeam/CV_WiseOtter_UnibeamBlue
   defaultValue=2
   parameterName=ColorCV
  }
  n15 = SphereMask(A=TextureCoordinate() {uTiling=0}, B=Constant2Vector() {g=0.5}) {attenuationRadius=0.5, hardnessPercent=0}
  n16 = AppendVector(A=DynamicParameter() {defaultValue=<rgba 2 0 1 1>, paramNames [UScale,Dissolve,Param3,Param4]}.UScale, B=Constant() {r=1})
  n17 = Panner(Coordinate=Multiply(A=TextureCoordinate() {uTiling=2}, B=@n16)) {speedX=-0.7}
  n18 = TextureCoordinate() {uTiling=1.5}
  n19 = TextureSample(UVs=Panner(Coordinate=Multiply(A=@n18, B=@n16)) {speedX=-0.15}) {samplerType=SAMPLERTYPE_Masks, texture=/Game/Weapons/WeaponSkins/Wraps/Textures/Mash/Mash_Noise}
  n20 = Multiply(A=LinearInterpolate(
    A=TextureSample(UVs=@n17) {samplerType=SAMPLERTYPE_Masks, texture=&Mash_Noise}.R
    B=Power(Base=@n19.R) {constExponent=1.5}
   )) {constB=0.3}
  n21 = Multiply(
   A=Power(Base=Add(A=@n15, B=@n20) {constB=0}) {constExponent=20}
   B=VectorParameter() {parameterName=Color1, defaultValue=<rgba 0.0677 0.3785 1 1>}.RGB
  )
  MaterialFunctionCall(Emissive=Add(A=Multiply(A=@n8) {constB=0.5}, B=Add(A=@n14, B=@n21) {constB=0})) {materialFunction=&MF_FX_EmissiveAdjustments}
  MaterialFunctionCall!(
   Particle Position
   Normalized Particle Direction
   Capsule Diameter
   Capsule Pivot Alignment
   Scale Along Velocity Vector
   ---------Scale Over Velocity Parameters---------
   Particle Speed
   Minimum Speed
   Max Speed
   Minimum Scale Factor
   Maximum Scale Factor
   ---------- Advanced -----------
   Cap Size Multiplier
   Use Camera Offset
   Use Advanced Features
   World Position to Negate
  ) {materialFunction=/Engine/Functions/Engine_MaterialFunctions02/WorldPositionOffset/Sprite_Capsule}
  n22 = ParticleColor()
  n23 = TextureCoordinate() {uTiling=0}
  n24 = Constant2Vector() {g=0.5}
  n25 = Subtract(
   A=Power(Base=SphereMask(A=@n23, B=@n24) {attenuationRadius=0.5, hardnessPercent=0}) {constExponent=1.5}
   B=DynamicParameter() {defaultValue=<rgba 2 0 1 1>, paramNames [UScale,Dissolve,Param3,Param4]}.Dissolve
  )
  n26 = ComponentMask(None=TextureCoordinate()) {g=true}
  n27 = TextureCoordinate()
  n28 = OneMinus(None=Sine(None=ComponentMask(None=@n27) {r=true}) {period=2})
  n29 = Multiply(A=Subtract(A=Sine(None=@n26) {period=2}, B=LinearInterpolate(Alpha=@n28) {constA=-2}) {constB=0}) {constB=5}
  n30 = TextureCoordinate()
  n31 = Constant2Vector() {g=0.5, r=0.5}
  Multiply(
   A=Multiply(A=Multiply(A=@n22.A, B=@n25), B=Saturate(None=@n29))
   B=Power(Base=SphereMask(A=@n30, B=@n31) {attenuationRadius=0.55, hardnessPercent=0})
  )
  n32 = TextureCoordinate()
  Power(Base=Sine(None=ComponentMask(None=@n32) {g=true}) {period=2})
  SphereMask(A=TextureCoordinate() {coordinateIndex=1, uTiling=0}, B=Constant2Vector() {g=0.5}) {attenuationRadius=0.5, hardnessPercent=0}
  SphereMask(A=TextureCoordinate() {coordinateIndex=1, vTiling=0}, B=Constant2Vector() {r=0.5}) {attenuationRadius=0.5, hardnessPercent=0}
  StaticBool()
 ]
}
