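# Unlit, two-sided translucent material for Niagara mesh particles, drawing the WiseOtter center muzzle effect.
# Radial UVs pan a blurred noise texture, which combines with a soft center sphere mask to drive brightness, tinted by particle color through MF_FX_EmissiveAdjustments.
# The same noise distorts the UVs of the muzzle texture, whose red channel is multiplied by particle alpha.
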Material /WiseOtter/Assets/FX/M_WiseOtter_CenterMuzzle
graph {
 properties {
  bAutomaticallySetUsageInEditor=false
  bUsedWithNiagaraMeshParticles=true
  blendMode=BLEND_Translucent
  shadingModel=MSM_Unlit
  twoSided=true
  usedShadingModels=Unlit
 }
 nodes [
  n1 = SphereMask(A=TextureCoordinate(), B=Constant() {r=0.5}) {attenuationRadius=0.2, hardnessPercent=0}
  n2 = MaterialFunctionCall(Vector or UVs=Subtract(A=TextureCoordinate() {uTiling=2, vTiling=2}, B=Constant() {r=0.95})) {materialFunction=&VectorToRadialValue}
  n3 = Constant2Vector() {g=0.7, r=3}
  n4 = TextureSample(UVs=Panner(Coordinate=Multiply(A=@n2.Radial Coordinates, B=@n3)) {speedY=-1}) {samplerType=SAMPLERTYPE_Grayscale, texture=&LowResBlurredNoise}
  n5 = LinearInterpolate(Alpha=Add(A=Power(Base=@n1) {constExponent=3}, B=@n4.R)) {constA=-0.05, constB=3}
  n6 = ParticleColor()
  MaterialFunctionCall(Emissive=Multiply(A=Multiply(A=@n5), B=@n6.RGB)) {materialFunction=&MF_FX_EmissiveAdjustments}
  Multiply()
  Multiply(A=Multiply(B=VertexNormalWS())) {constB=150}
  n7 = TextureCoordinate()
  n8 = Multiply(A=@n4.R) {constB=0.15}
  Multiply(A=@n6.A, B=TextureSample(UVs=Add(A=@n7, B=@n8)) {texture=&/T_WiseOtter_CenterMuzzle}.R)
  Power(Base=OneMinus(None=@n1)) {constExponent=1.2}
 ]
}
