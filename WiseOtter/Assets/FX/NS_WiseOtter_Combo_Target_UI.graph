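# Niagara sprite system for a combo target reticle UI, driven by a LockProgress user parameter.
# Two bursts spawn sprites in four partitions rotated in 90-degree steps and held at camera-relative scale.
# Each partition's timeline triggers from LockProgress offset by quarter steps, while a system-timeline curve pulses sprite size.
# Material parameter (BulletCount-driven), pivot and color-fade steps are disabled, leaving the per-partition timeline without an active visual consumer.
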NiagaraSystem /WiseOtter/Assets/FX/NS_WiseOtter_Combo_Target_UI
dependencies {
 usedDataInterfaces [&NiagaraDataInterfaceCurve, &NiagaraDataInterfaceArrayFloat]
 usedDynamicInputs []
 usedModules [&SystemState]
}
emitters {Targeted_Reticle001 {
  emitterUpdateScript [
   &EmitterState(
    Inactive Response=::NewEnumerator1
    Life Cycle Mode=::NewEnumerator1
    Loop Behavior=::NewEnumerator1
    Loop Duration Mode=::NewEnumerator1
    UseLoopDelay=false
   )
   &SpawnBurst_Instantaneous(Use Spawn Probability=false, UseLoopCountLimit=true)
   &SpawnBurst_Instantaneous(Spawn Group=1, Use Spawn Probability=false, UseLoopCountLimit=true) {module=SpawnBurst_Instantaneous001}
  ]
  particleSpawnScript [
   &PartitionParticles(Number Of Partitions=4)
   &InitializeParticle(
    Sprite Rotation Angle=&Multiply_FloatByInt(Float=90, Integer=&Add_Integer(A=2, B=$Particles.PartitionParticles.Partition))
    Sprite Rotation Mode=::NewEnumerator0
    Sprite Size Mode=::NewEnumerator3
    Uniform Sprite Size=175
    UsePositionOffset=false
    Write Material Random=true
   )
  ]
  particleUpdateScript [
   &ParticleState(Kill Particles When Lifetime Has Elapsed=false)
   /Niagara/Modules/Update/Camera/MaintainInCameraParticleScale(
    Pixel Depth:NiagaraFloat=&Add_Float(A=$Engine.Owner.LODDistance, B=100)
    Render Target Width:NiagaraFloat=&Multiply_Float(A=2560, B=1)
   )
   &Timeline(
    Play=&SetBoolByFloatComparison(
     A=&Subtract_Float(A=$User.LockProgress, B=&Multiply_FloatByInt(Float=0.25, Integer=$Particles.PartitionParticles.Partition))
     B=0.175
    )
    Timeline End=0.15
   )
   &DynamicMaterialParameters(
    Index 0 Param 1=&MaskFloatBySpawnGroup(
     Masked Float=&MakeCustomFloatFromBool(
      Bool=&SetBoolByIntComparison(A=$User.BulletCount, B=$Particles.Initial.PartitionParticles.Partition, Comparison Type=::NewEnumerator0)
      True Float=0
      False Float=1
     )
     Passthrough Float=0
    )
    Module.Param0WriteEnabled=true
    Module.Param1WriteEnabled=false
    Module.Param2WriteEnabled=false
    Module.Param3WriteEnabled=false
   ) {enabled=false}
   /SetVariables(Particles.PivotOffset:Vector2f=&Lerp_Vector2(
     A=<1.3 1.3>
     B=<2.5 2.5>
     Alpha=&FloatFromCurve(
      FloatCurve=DataInterface() {propertyValues {Curve {keys [{interpMode=RCIM_Cubic, tangentMode=RCTM_User, time=0, value=1}, {interpMode=RCIM_Cubic, tangentMode=RCTM_User, time=0.66, value=0}]}}}
      CurveIndex=$Particles.Timeline.NormalizedTime
      Scale Curve=&MaskFloatBySpawnGroup(Masked Float=0, Passthrough Float=1)
     )
    )) {enabled=false, isSetParameters=true}
   /SetVariables(Particles.SpriteSize=&Multiply_Vector2D_ByFloat(Float=&Lerp_Float(A=2, B=1, Alpha=$System.Timeline.NormalizedTime), Vector2D=$Particles.SpriteSize)) {enabled=false, isSetParameters=true}
   /SetVariables(Particles.SpriteSize=/Niagara/DynamicInputs/Add/Add_Vector2(
     A:Vector2f=$Particles.SpriteSize
     B:Vector2f=&Vector2DFromFloat(Value=&FloatFromCurve(
       FloatCurve=DataInterface() {propertyValues {Curve {keys [
           {arriveTangent=9.0725, interpMode=RCIM_Cubic, leaveTangent=9.0725, tangentMode=RCTM_User, time=0, value=0}
           {arriveTangent=-7.1386, interpMode=RCIM_Cubic, leaveTangent=-7.1386, tangentMode=RCTM_User, time=0.2, value=1.25}
           {arriveTangent=-0.0405, interpMode=RCIM_Cubic, leaveTangent=-0.0405, tangentMode=RCTM_User, time=0.6, value=0}
          ]}}}
       CurveIndex=$System.Timeline.Time
       Scale Curve=50
      ))
    )) {isSetParameters=true}
   &ScaleColor(
    Scale Alpha=&FloatFromCurve(
     FloatCurve=DataInterface() {propertyValues {Curve {keys [
         {arriveTangent=5.6222, interpMode=RCIM_Cubic, time=0, value=0}
         {interpMode=RCIM_Cubic, tangentMode=RCTM_User, time=0.2, value=1}
         {arriveTangent=-0.0405, interpMode=RCIM_Cubic, leaveTangent=-0.0405, tangentMode=RCTM_User, time=1, value=1}
        ]}}}
     CurveIndex=$Emitter.LoopedAge
     Scale Curve=&MaskFloatBySpawnGroup(
      Masked Float=1
      Passthrough Float=&FloatFromCurve(
       FloatCurve=DataInterface() {propertyValues {Curve {keys [
           {interpMode=RCIM_Cubic, tangentMode=RCTM_User, time=0, value=0}
           {interpMode=RCIM_Cubic, tangentMode=RCTM_User, time=0.5, value=1}
           {interpMode=RCIM_Cubic, tangentMode=RCTM_User, time=0.75, value=1}
           {tangentMode=RCTM_User, time=1, value=0}
          ]}}}
       CurveIndex=$Particles.Timeline.NormalizedTime
      )
     )
    )
    Scale RGB=&Lerp_Vector(A=<1 1 1>, B=<1 1 1>, Alpha=$System.LockProgress)
    ScaleA=true
    ScaleRGB=true
   ) {enabled=false}
  ]
  rendererClasses [&NiagaraSpriteRendererProperties]
  renderers [{rendererClass=&NiagaraSpriteRendererProperties}]
 }}
summary {
 systemName=NS_WiseOtter_Combo_Target_UI
 emitters [{emitterName=Targeted_Reticle001, rendererClasses [&NiagaraSpriteRendererProperties]}]
 userVariables [
  User.BulletCount {type=NiagaraInt32}
  User.LockProgress {type=NiagaraFloat, defaultValue=0, description=""}
 ]
}
